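# Translucent parent material for the SureBamboo gadget prop, built on the shared 2023 anime character shader with Cloth shading.
# The anime Pre, Calculations and Post functions chain into the material attributes, with a tinted Fresnel glow for emissive.
# Opacity takes the minimum of a parameter, local-height gradient and camera-distance fade, routed through MF_SortTranslucency.
# Translucent custom depth writes and velocity output are enabled.
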Material /BRCosmetics/Animation/Props/Gadgets_SureBamboo/Materials/M_Gadget_SureBamboo_Parent
graph {
 properties {
  allowTranslucentCustomDepthWrites=true
  bAutomaticallySetUsageInEditor=false
  bOutputTranslucentVelocity=true
  bUseMaterialAttributes=true
  bUsedWithMeshDeformer=true
  bUsedWithMorphTargets=true
  bUsedWithSkeletalMesh=true
  bUsedWithStaticLighting=true
  blendMode=BLEND_Translucent
  pixelDepthOffsetMode=PDOM_Legacy
  shadingModel=MSM_Cloth
  usedShadingModels=Cloth
 }
 nodes [
  n1 = Min(A=ScalarParameter() {defaultValue=0.7, group=04_OpacityControl, parameterName=Opactiy_Value, sliderMax=1}, B=NamedRerouteUsage())
  n2 = NamedRerouteUsage()
  NamedRerouteDeclaration(None=MaterialFunctionCall!(Opacity=Min(A=@n1, B=@n2)) {materialFunction=/Game/Global/Materials/MaterialGym/Materials/SortedOutline/MF_SortTranslucency}.Result) {name=OpasityMask, nodeColor=<rgba 1 0.3984 0 1>}
  n3 = ComponentMask(None=ObjectBounds()) {b=true}
  n4 = Divide(A=Add(A=LocalPosition().Z, B=@n3), B=@n3)
  n5 = ScalarParameter() {defaultValue=2.2, group=04_OpacityControl, parameterName=ZGradient_Offset, sliderMax=4, sliderMin=-4}
  NamedRerouteDeclaration(None=Divide(A=Subtract(A=@n4, B=@n5), B=ScalarParameter() {defaultValue=0.7, group=04_OpacityControl, parameterName=ZGradient_Scale, sliderMax=4})) {name=ZGradient, nodeColor=<rgba 0.0781 0 1 1>}
  n6 = WorldPosition() {worldPositionShaderOffset=WPT_ExcludeAllShaderOffsets}
  n7 = CameraPositionWS()
  n8 = Divide(
   A=Add(A=Distance(A=@n6.XYZ, B=@n7), B=ScalarParameter() {defaultValue=50, group=04_OpacityControl, parameterName=CameraFade_Offset})
   B=ScalarParameter() {defaultValue=100, group=04_OpacityControl, parameterName=CameraFade_Distance}
  )
  n9 = ScalarParameter() {defaultValue=4, group=04_OpacityControl, parameterName=CameraFade_Exp}
  NamedRerouteDeclaration(None=Saturate(None=Power(Base=@n8, Exp=@n9))) {name=CameraFading, nodeColor=<rgba 0 0.8828 1 1>}
  n10 = VectorParameter() {group=04_OpacityControl, parameterName=Glow_Color, defaultValue=<rgba 0 1 0.792 5>}
  n11 = Multiply(A=@n10.RGB, B=@n10.RGB)
  n12 = ScalarParameter() {group=04_OpacityControl, parameterName=Glow_Blend, sliderMax=1}
  n13 = ScalarParameter() {defaultValue=5, group=04_OpacityControl, parameterName=Glow_FresnelExponent, sliderMax=8}
  NamedRerouteDeclaration(None=Multiply(A=Multiply(A=@n11, B=@n12), B=Fresnel(ExponentIn=@n13))) {name=GlowColor, nodeColor=<rgba 0 0.9766 1 1>}
  n14 = MaterialFunctionCall!(CustomUV) {materialFunction=/Game/Global/Materials/Character_Parent_2023/Anime_Parent_2023/MaterialFunctions/MFC_AnimeShader_2023_Pre}
  n15 = MaterialFunctionCall!(CustomLightVector, OverrideNormal_WS, STT=@n14.STT, Normal_Map=@n14.STT, DFL=@n14.STT, FX=@n14.STT) {materialFunction=/Game/Global/Materials/Character_Parent_2023/Anime_Parent_2023/MaterialFunctions/MFC_AnimeShader_2023_Calculations}
  n16 = NamedRerouteUsage()
  SetMaterialAttributes!(
   MaterialAttributes=MaterialFunctionCall!(
    Shading_Gradient=@n15.Shading_Gradient
    Threshold_Gradient=@n15.Shading_Gradient
    Specular_Gradient=@n15.Shading_Gradient
    Sheen_Mask=@n15.Shading_Gradient
    Thickness=@n15.Shading_Gradient
    DFL_Mask=@n15.Shading_Gradient
    FX=@n15.Shading_Gradient
    Metallic_Gradient=@n15.Shading_Gradient
    RimLight_Gradient=@n15.Shading_Gradient
    NormalBuffer=@n15.Shading_Gradient
    --------
    CINE_LightSources=@n15.Shading_Gradient
    CINE_LightColor=@n15.Shading_Gradient
    ---------
    Color_Lit=@n14.Color_Lit
    Color_Shaded=@n14.Color_Lit
    SRM=@n14.Color_Lit
    Emissive=Add(A=@n14.Emissive, B=@n16)
    DFL_Color=@n14.Color_Lit
    Specular_Color=@n14.Color_Lit
    RimLight_Color=@n14.Color_Lit
   ) {materialFunction=/Game/Global/Materials/Character_Parent_2023/Anime_Parent_2023/MaterialFunctions/MFC_AnimeShader_2023_Post}.Result
   Opacity=NamedRerouteUsage()
  ) {attributeSetTypes [#1]}
 ]
}
